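# Animation Blueprint for the Season 39 glider, animating its deploy, glide and close.
# The Parachute state machine follows skydiving state and bForceOpen, choosing front-end or in-game sequences by IsCosmeticPreview.
# A curve-weighted jitter additive, then SideSpin and TopSpin additive state machines layer on top before DefaultSlot.
# A directional Lean Transitions machine with nested lean-in states feeds an additive node that is not connected to the output pose.
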AnimBlueprint /BRCosmetics/Gadgets/Assets/VinderTech_GliderChute/Glider_Season_39/Meshes/Season_39_Glider_AnimBP
target_skeleton=&/Season_39_Glider_Skeleton
graphs [
 AnimGraph {nodes [
   AnimGraphNode_StateMachine_1 = Lean Transitions() {is=title AnimGraphNode_StateMachine}
   Apply Additive!(Base, Additive=@AnimGraphNode_StateMachine_1, AlphaCurveName:Name=ParaGlide_Lean) -> (Pose) {alphaInputType=Curve, alphaBoolBlend {blendOption=Linear}, alphaScaleBiasClamp {bias=1, scale=-1}}
   n1 = Get Jitter_Speed() -> (Jitter_Speed:double)
   AnimGraphNode_StateMachine_0 = Parachute() {is=title AnimGraphNode_StateMachine}
   AnimGraphNode_StateMachine_3 = SideSpin() {is=title AnimGraphNode_StateMachine}
   n2 = Apply Additive!(
    Base=Apply Additive!(
     Base=@AnimGraphNode_StateMachine_0
     Additive="Sequence Player (additive) 'Season39_Jitter_GLIDER'"!(PlayRate:float=@n1) -> (Pose) {sequence=/BRCosmetics/Gadgets/Assets/VinderTech_GliderChute/Glider_Season_39/Animation/Season39_Jitter_GLIDER}
     AlphaCurveName:Name=ParaGlide_Jitter
    ) -> (Pose) {alphaInputType=Curve, alphaBoolBlend {blendOption=Linear}, alphaScaleBiasClamp {bias=1, scale=-1}}
    Additive=@AnimGraphNode_StateMachine_3
    bAlphaBoolEnabled:bool=True
   ) -> (Pose) {alphaInputType=Bool, alphaBoolBlend {blendOption=Linear}, alphaScaleBiasClamp {bias=1, scale=-1}}
   AnimGraphNode_StateMachine_2 = TopSpin() {is=title AnimGraphNode_StateMachine}
   Output Pose(Result=Slot 'DefaultSlot'(Source=Apply Additive!(Base=@n2, Additive=@AnimGraphNode_StateMachine_2, bAlphaBoolEnabled:bool=True) -> (Pose) {alphaInputType=Bool, alphaBoolBlend {blendOption=Linear}, alphaScaleBiasClamp {bias=1, scale=-1}})) {layerGroup=None}
  ]}
 Parachute {
  entry=Closed
  path=AnimGraph.AnimGraphNode_StateMachine_0.Parachute
  ids=id
  nodes [
   AnimStateNode_0 = Parachute_Into() {is=title AnimStateNode}
   AnimStateNode_1 = Idle()
   AnimStateNode_2 = Closed() {is=title AnimStateNode}
   AnimStateTransitionNode_0 = AnimStateTransitionNode() {bAutomaticRuleBasedOnSequencePlayerInState=true, crossfadeDuration=0}
   AnimStateTransitionNode_1 = AnimStateTransitionNode()
   AnimStateTransitionNode_2 = AnimStateTransitionNode() {crossfadeDuration=0}
   AnimStateTransitionNode_3 = AnimStateTransitionNode() {crossfadeDuration=0}
  ]
  links [
   {
    dst=Parachute_Into
    rule=AnimGraph.AnimGraphNode_StateMachine_0.Parachute.AnimStateTransitionNode_1.AnimationTransitionGraph_1
    src=Closed
    via=AnimStateTransitionNode_1
   }
   {
    dst=Closed
    rule=AnimGraph.AnimGraphNode_StateMachine_0.Parachute.AnimStateTransitionNode_2.AnimationTransitionGraph_1
    src=Idle
    via=AnimStateTransitionNode_2
   }
   {
    dst=Closed
    rule=AnimGraph.AnimGraphNode_StateMachine_0.Parachute.AnimStateTransitionNode_3.AnimationTransitionGraph_1
    src=Parachute_Into
    via=AnimStateTransitionNode_3
   }
   {
    dst=Idle
    rule=AnimGraph.AnimGraphNode_StateMachine_0.Parachute.AnimStateTransitionNode_0.AnimationTransitionGraph_1
    src=Parachute_Into
    via=AnimStateTransitionNode_0
   }
  ]
 }
 Parachute_Into {
  path=AnimGraph.AnimGraphNode_StateMachine_0.Parachute.AnimStateNode_0.Parachute_Into
  nodes [
   n1 = Sequence Player 'Season39_FrontEnd_Into_GLIDER'() {bLoopAnimation=false, sequence=/BRCosmetics/Gadgets/Assets/VinderTech_GliderChute/Glider_Season_39/Animation/Season39_FrontEnd_Into_GLIDER}
   n2 = Get OwnerGlider() -> (OwnerGlider:Fort Player Parachute Object Reference)
   Output Animation Pose(Result=Blend Poses by bool(
     BlendPose_0=@n1
     BlendPose_1=Sequence Player 'Season39_Into_GLIDER'() {bLoopAnimation=false, sequence=/BRCosmetics/Gadgets/Assets/VinderTech_GliderChute/Glider_Season_39/Animation/Season39_Into_GLIDER}
     bActiveValue=Get IsCosmeticPreview!(self:Fort Player Parachute Object Reference=@n2) -> (IsCosmeticPreview:bool)
    ) {blendType=Linear})
  ]
 }
 Idle {
  path=AnimGraph.AnimGraphNode_StateMachine_0.Parachute.AnimStateNode_1.Idle
  nodes [
   n1 = Sequence Player 'Season_39_Base_Pose_GLIDER'() {bLoopAnimation=false, sequence=/BRCosmetics/Gadgets/Assets/VinderTech_GliderChute/Glider_Season_39/Animation/Season_39_Base_Pose_GLIDER}
   n2 = Get OwnerGlider()
   Output Animation Pose(Result=Blend Poses by bool(
     BlendPose_0=Sequence Player 'Season39_FrontEnd_GLIDER'() {sequence=/BRCosmetics/Gadgets/Assets/VinderTech_GliderChute/Glider_Season_39/Animation/Season39_FrontEnd_GLIDER}
     BlendPose_1=@n1
     bActiveValue=Get IsCosmeticPreview(self=@n2)
    ) {blendType=Linear})
  ]
 }
 Closed {
  path=AnimGraph.AnimGraphNode_StateMachine_0.Parachute.AnimStateNode_2.Closed
  nodes [
   n1 = Sequence Player 'Season39_FrontEnd_Close_GLIDER'() {bLoopAnimation=false, sequence=/BRCosmetics/Gadgets/Assets/VinderTech_GliderChute/Glider_Season_39/Animation/Season39_FrontEnd_Close_GLIDER}
   n2 = Get OwnerGlider()
   Output Animation Pose(Result=Blend Poses by bool(
     BlendPose_0=@n1
     BlendPose_1=Sequence Player 'Season39_Closed_GLIDER'() {sequence=/BRCosmetics/Gadgets/Assets/VinderTech_GliderChute/Glider_Season_39/Animation/Season39_Closed_GLIDER}
     bActiveValue=Get IsCosmeticPreview(self=@n2)
    ) {blendType=Linear})
  ]
 }
 AnimationTransitionGraph_1 {
  path=AnimGraph.AnimGraphNode_StateMachine_0.Parachute.AnimStateTransitionNode_0.AnimationTransitionGraph_1
  nodes [Result(bCanEnterTransition="Equal (Float)"!(A:double="Get Relevant Anim Time Remaining (Parachute_Into)"!() -> (ReturnValue:float) {is=title K2Node_AnimGetter}, B:double=0) -> (ReturnValue:bool) {is=title K2Node_CallFunction})]
 }
 AnimationTransitionGraph_1 {
  path=AnimGraph.AnimGraphNode_StateMachine_0.Parachute.AnimStateTransitionNode_1.AnimationTransitionGraph_1
  nodes [
   n1 = Get SkydivingState() -> (SkydivingState:Fort Anim Input Skydiving Structure)
   Result(bCanEnterTransition=OR Boolean(
     A=Break Fort Anim Input Skydiving!(FortAnimInput_Skydiving:"Fort Anim Input Skydiving Structure (by ref)"=@n1) -> (
      ParaGilderRootOffset:Vector=~
      ParaGliderRootRotationOverride:Rotator=~
      bUseParaGlideRootModifier:bool=~
      bIsSkydivingFromLaunchPad:bool=~
      bIsSkydivingFromBus:bool=~
      bIsInVortex:bool=~
      bIsDiving:bool=~
      bIsParachuteOpen:bool=~
      bIsSkydiving:bool=~
      bIsParachuteLeaning:bool=~
      bIsSkydiveLeaning:bool=~
      bIsLeaning:bool=~
      bIsSkydiveDiveMode:bool=~
      bPlaySkydiveDrift:bool=~
      bSkydiveDriftDelayActive:bool=~
      bSkydiveDriftAnimAllowed:bool=~
      bIsGliderRight:bool=~
      bIsGliderCenter:bool=~
      bIsGliderLeft:bool=~
      bIsGliderForward:bool=~
      bIsGliderBack:bool=~
      bHasDeployedAndIsStandingUpright:bool=~
      LocalAccelForward:float=~
      LocalAccelRight:float=~
      LocalVelocityRight:float=~
      SkydiveAimPitch:float=~
      SkydiveAimYaw:float=~
      DeployChuteAnimRate:float=~
      SkydiveDriftAnimRate:float=~
      SkydiveDriftAnimRateCurrent:float=~
      SkydiveFidgetAnimRate:float=~
      SkydiveFidgetAnimRateCurrent:float=~
      SkydiveAdditiveAlpha:float=~
      SkydiveDriftDelay:float=~
      SkydiveDriftAnim:int
      SkydiveDriftAnimMax:int
      LaunchpadAnim:int
      LocalAccelDir:ESkydivingDirection Enum
      DirectionLast:ESkydivingDirection Enum
      PoseAdditive:Anim Sequence Object Reference
     ) {is=title K2Node_BreakStruct}.bIsParachuteOpen
     B=Get bForceOpen() -> (bForceOpen:bool)
    ))
   n2 = Break Fort Anim Input Skydiving(FortAnimInput_Skydiving=Get SkydivingState()) -> (
    ParaGilderRootOffset=~
    ParaGliderRootRotationOverride=~
    bUseParaGlideRootModifier=~
    bIsSkydivingFromLaunchPad=~
    bIsSkydivingFromBus=~
    bIsInVortex=~
    bIsDiving=~
    bIsParachuteOpen=~
    bIsSkydiving=~
    bIsParachuteLeaning=~
    bIsSkydiveLeaning=~
    bIsLeaning=~
    bIsSkydiveDiveMode=~
    bPlaySkydiveDrift=~
    bSkydiveDriftDelayActive=~
    bSkydiveDriftAnimAllowed=~
    bIsGliderRight=~
    bIsGliderCenter=~
    bIsGliderLeft=~
    bIsGliderForward=~
    bIsGliderBack=~
    bHasDeployedAndIsStandingUpright=~
    LocalAccelForward=~
    LocalAccelRight=~
    LocalVelocityRight=~
    SkydiveAimPitch=~
    SkydiveAimYaw=~
    DeployChuteAnimRate=~
    SkydiveDriftAnimRate=~
    SkydiveDriftAnimRateCurrent=~
    SkydiveFidgetAnimRate=~
    SkydiveFidgetAnimRateCurrent=~
    SkydiveAdditiveAlpha=~
    SkydiveDriftDelay=~
   )
   n3 = Get OwnerGlider()
   AND Boolean(
    A=OR Boolean!(A=@n2.bIsSkydivingFromLaunchPad, B=@n2.bIsSkydivingFromBus, C:bool=@n2.bIsInVortex, D:bool=Get IsCosmeticPreview(self=@n3)) -> (ReturnValue)
    B=@n2.bIsParachuteOpen
   )
  ]
 }
 AnimationTransitionGraph_1 {
  path=AnimGraph.AnimGraphNode_StateMachine_0.Parachute.AnimStateTransitionNode_2.AnimationTransitionGraph_1
  nodes [
   n1 = Get bForceOpen()
   n2 = Break Fort Anim Input Skydiving(FortAnimInput_Skydiving=Get SkydivingState()) -> (
    ParaGilderRootOffset=~
    ParaGliderRootRotationOverride=~
    bUseParaGlideRootModifier=~
    bIsSkydivingFromLaunchPad=~
    bIsSkydivingFromBus=~
    bIsInVortex=~
    bIsDiving=~
    bIsParachuteOpen=~
    bIsSkydiving=~
    bIsParachuteLeaning=~
    bIsSkydiveLeaning=~
    bIsLeaning=~
    bIsSkydiveDiveMode=~
    bPlaySkydiveDrift=~
    bSkydiveDriftDelayActive=~
    bSkydiveDriftAnimAllowed=~
    bIsGliderRight=~
    bIsGliderCenter=~
    bIsGliderLeft=~
    bIsGliderForward=~
    bIsGliderBack=~
    bHasDeployedAndIsStandingUpright=~
    LocalAccelForward=~
    LocalAccelRight=~
    LocalVelocityRight=~
    SkydiveAimPitch=~
    SkydiveAimYaw=~
    DeployChuteAnimRate=~
    SkydiveDriftAnimRate=~
    SkydiveDriftAnimRateCurrent=~
    SkydiveFidgetAnimRate=~
    SkydiveFidgetAnimRateCurrent=~
    SkydiveAdditiveAlpha=~
    SkydiveDriftDelay=~
   )
   Result(bCanEnterTransition=AND Boolean(A=NOT Boolean(A=@n2.bIsParachuteOpen), B=NOT Boolean(A=@n1)))
   NOT Boolean(A=Get IsParachuteOpen() -> (IsParachuteOpen:bool))
  ]
 }
 AnimationTransitionGraph_1 {
  path=AnimGraph.AnimGraphNode_StateMachine_0.Parachute.AnimStateTransitionNode_3.AnimationTransitionGraph_1
  nodes [
   n1 = Break Fort Anim Input Skydiving(FortAnimInput_Skydiving=Get SkydivingState()) -> (
    ParaGilderRootOffset=~
    ParaGliderRootRotationOverride=~
    bUseParaGlideRootModifier=~
    bIsSkydivingFromLaunchPad=~
    bIsSkydivingFromBus=~
    bIsInVortex=~
    bIsDiving=~
    bIsParachuteOpen=~
    bIsSkydiving=~
    bIsParachuteLeaning=~
    bIsSkydiveLeaning=~
    bIsLeaning=~
    bIsSkydiveDiveMode=~
    bPlaySkydiveDrift=~
    bSkydiveDriftDelayActive=~
    bSkydiveDriftAnimAllowed=~
    bIsGliderRight=~
    bIsGliderCenter=~
    bIsGliderLeft=~
    bIsGliderForward=~
    bIsGliderBack=~
    bHasDeployedAndIsStandingUpright=~
    LocalAccelForward=~
    LocalAccelRight=~
    LocalVelocityRight=~
    SkydiveAimPitch=~
    SkydiveAimYaw=~
    DeployChuteAnimRate=~
    SkydiveDriftAnimRate=~
    SkydiveDriftAnimRateCurrent=~
    SkydiveFidgetAnimRate=~
    SkydiveFidgetAnimRateCurrent=~
    SkydiveAdditiveAlpha=~
    SkydiveDriftDelay=~
   )
   n2 = Get bForceOpen()
   Result(bCanEnterTransition=AND Boolean(A=NOT Boolean(A=@n1.bIsParachuteOpen), B=NOT Boolean(A=@n2)))
  ]
 }
 Lean Transitions {
  entry=Lean_Zero
  path=AnimGraph.AnimGraphNode_StateMachine_1.Lean Transitions
  ids=id
  nodes [
   AnimStateNode_0 = Lean_Zero() {
    is=title AnimStateNode
    stateEntered {guid=#1}
    stateFullyBlended {guid=#2}
    stateLeft {guid=#3}
   }
   AnimStateNode_2 = Lean_Out() {
    is=title AnimStateNode
    stateEntered {guid=#4}
    stateFullyBlended {guid=#5}
    stateLeft {guid=#6}
   }
   AnimStateNode_5 = Lean_Into() {
    is=title AnimStateNode
    stateEntered {guid=#7}
    stateFullyBlended {guid=#8}
    stateLeft {guid=#9}
   }
   AnimStateTransitionNode_0 = AnimStateTransitionNode()
   AnimStateTransitionNode_11 = AnimStateTransitionNode() {
    transitionEnd {guid=#10}
    transitionInterrupt {guid=#11}
    transitionStart {guid=#12}
   }
   AnimStateTransitionNode_18 = AnimStateTransitionNode() {
    bAutomaticRuleBasedOnSequencePlayerInState=true
    blendMode=Linear
    priorityOrder=2
    transitionEnd {guid=#13}
    transitionInterrupt {guid=#14}
    transitionStart {guid=#15}
   }
   AnimStateTransitionNode_19 = AnimStateTransitionNode() {
    blendMode=Linear
    transitionEnd {guid=#16}
    transitionInterrupt {guid=#17}
    transitionStart {guid=#18}
   }
  ]
  links [
   {
    dst=Lean_Out
    rule=AnimGraph.AnimGraphNode_StateMachine_1.Lean Transitions.AnimStateTransitionNode_11.AnimationTransitionGraph_0
    src=Lean_Into
    via=AnimStateTransitionNode_11
   }
   {
    dst=Lean_Zero
    rule=AnimGraph.AnimGraphNode_StateMachine_1.Lean Transitions.AnimStateTransitionNode_18.AnimationTransitionGraph_0
    src=Lean_Out
    via=AnimStateTransitionNode_18
   }
   {
    dst=Lean_Zero
    rule=AnimGraph.AnimGraphNode_StateMachine_1.Lean Transitions.AnimStateTransitionNode_19.AnimationTransitionGraph_0
    src=Lean_Out
    via=AnimStateTransitionNode_19
   }
   {
    dst=Lean_Into
    rule=AnimGraph.AnimGraphNode_StateMachine_1.Lean Transitions.AnimStateTransitionNode_0.AnimationTransitionGraph_0
    src=Lean_Zero
    via=AnimStateTransitionNode_0
   }
  ]
 }
 Lean_Zero {
  path=AnimGraph.AnimGraphNode_StateMachine_1.Lean Transitions.AnimStateNode_0.Lean_Zero
  nodes [Output Animation Pose(Result=Additive Identity Pose())]
 }
 Lean_Out {
  path=AnimGraph.AnimGraphNode_StateMachine_1.Lean Transitions.AnimStateNode_2.Lean_Out
  nodes [
   n1 = "Sequence Player (additive) 'Season39_Lean_Forward_Out'"() {bLoopAnimation=false, sequence=/BRCosmetics/Gadgets/Assets/VinderTech_GliderChute/Glider_Season_39/Animation/Season39_Lean_Forward_Out}
   n2 = "Sequence Player (additive) 'Season39_Lean_Right_Out'"() {bLoopAnimation=false, sequence=/BRCosmetics/Gadgets/Assets/VinderTech_GliderChute/Glider_Season_39/Animation/Season39_Lean_Right_Out}
   n3 = "Blend Poses (ESkydivingDirection)"!(
    BlendPose_0:Pose Link Structure
    BlendPose_1:Pose Link Structure=@n1
    BlendPose_2:Pose Link Structure="Sequence Player (additive) 'Season39_Lean_Back_Out'"() {bLoopAnimation=false, sequence=/BRCosmetics/Gadgets/Assets/VinderTech_GliderChute/Glider_Season_39/Animation/Season39_Lean_Back_Out}
    BlendPose_3:Pose Link Structure="Sequence Player (additive) 'Season39_Lean_Left_Out'"() {bLoopAnimation=false, sequence=/BRCosmetics/Gadgets/Assets/VinderTech_GliderChute/Glider_Season_39/Animation/Season39_Lean_Left_Out}
    BlendPose_4:Pose Link Structure=@n2
    BlendTime_0:float=0.5
    BlendTime_1:float=0.5
    BlendTime_2:float=0.5
    BlendTime_3:float=0.5
    BlendTime_4:float=0.5
    ActiveEnumValue:Byte=Break Fort Anim Input Skydiving(FortAnimInput_Skydiving=Get SkydivingState()) -> (
     ParaGilderRootOffset=~
     ParaGliderRootRotationOverride=~
     bUseParaGlideRootModifier=~
     bIsSkydivingFromLaunchPad=~
     bIsSkydivingFromBus=~
     bIsInVortex=~
     bIsDiving=~
     bIsParachuteOpen=~
     bIsSkydiving=~
     bIsParachuteLeaning=~
     bIsSkydiveLeaning=~
     bIsLeaning=~
     bIsSkydiveDiveMode=~
     bPlaySkydiveDrift=~
     bSkydiveDriftDelayActive=~
     bSkydiveDriftAnimAllowed=~
     bIsGliderRight=~
     bIsGliderCenter=~
     bIsGliderLeft=~
     bIsGliderForward=~
     bIsGliderBack=~
     bHasDeployedAndIsStandingUpright=~
     LocalAccelForward=~
     LocalAccelRight=~
     LocalVelocityRight=~
     SkydiveAimPitch=~
     SkydiveAimYaw=~
     DeployChuteAnimRate=~
     SkydiveDriftAnimRate=~
     SkydiveDriftAnimRateCurrent=~
     SkydiveFidgetAnimRate=~
     SkydiveFidgetAnimRateCurrent=~
     SkydiveAdditiveAlpha=~
     SkydiveDriftDelay=~
    ).DirectionLast
   ) -> (Pose:Pose Link Structure) {
    is=title AnimGraphNode_BlendListByEnum
    blendType=Linear
    blendPose [
     {linkId=-1, sourceLinkId=-1}
     {linkId=-1, sourceLinkId=-1}
     {linkId=-1, sourceLinkId=-1}
     {linkId=-1, sourceLinkId=-1}
     {linkId=-1, sourceLinkId=-1}
    ]
    blendTime [0.1,0.1,0.1,0.1,0.1]
   }
   Output Animation Pose(Result=@n3)
  ]
 }
 Lean_Into {
  path=AnimGraph.AnimGraphNode_StateMachine_1.Lean Transitions.AnimStateNode_5.Lean_Into
  nodes [
   AnimGraphNode_StateMachine_0 = Lean_Into() {is=title AnimGraphNode_StateMachine}
   Output Animation Pose(Result=@AnimGraphNode_StateMachine_0)
  ]
 }
 Lean_Into {
  entry=Zero
  path=AnimGraph.AnimGraphNode_StateMachine_1.Lean Transitions.AnimStateNode_5.Lean_Into.AnimGraphNode_StateMachine_0.Lean_Into
  ids=id
  nodes [
   AnimStateNode_0 = Zero() {is=title AnimStateNode}
   AnimStateNode_1 = Lean_Forward_Into() {is=title AnimStateNode}
   AnimStateNode_2 = Lean_Back_Into() {is=title AnimStateNode}
   AnimStateNode_3 = Lean_Left_Into() {is=title AnimStateNode}
   AnimStateNode_4 = Lean_Right_Into() {is=title AnimStateNode}
   AnimStateTransitionNode_0 = AnimStateTransitionNode()
   AnimStateTransitionNode_1 = AnimStateTransitionNode()
   AnimStateTransitionNode_10 = AnimStateTransitionNode() {crossfadeDuration=0.5}
   AnimStateTransitionNode_11 = AnimStateTransitionNode() {crossfadeDuration=0.5}
   AnimStateTransitionNode_12 = AnimStateTransitionNode() {crossfadeDuration=0.5}
   AnimStateTransitionNode_13 = AnimStateTransitionNode() {crossfadeDuration=0.5}
   AnimStateTransitionNode_14 = AnimStateTransitionNode() {crossfadeDuration=0.5}
   AnimStateTransitionNode_15 = AnimStateTransitionNode() {crossfadeDuration=0.5}
   AnimStateTransitionNode_2 = AnimStateTransitionNode()
   AnimStateTransitionNode_3 = AnimStateTransitionNode()
   AnimStateTransitionNode_4 = AnimStateTransitionNode() {crossfadeDuration=0.5}
   AnimStateTransitionNode_5 = AnimStateTransitionNode() {crossfadeDuration=0.5}
   AnimStateTransitionNode_6 = AnimStateTransitionNode() {crossfadeDuration=0.5}
   AnimStateTransitionNode_7 = AnimStateTransitionNode() {crossfadeDuration=0.5}
   AnimStateTransitionNode_8 = AnimStateTransitionNode() {crossfadeDuration=0.5}
   AnimStateTransitionNode_9 = AnimStateTransitionNode() {crossfadeDuration=0.5}
  ]
  links [
   {
    dst=Lean_Forward_Into
    rule=AnimGraph.AnimGraphNode_StateMachine_1.Lean Transitions.AnimStateNode_5.Lean_Into.AnimGraphNode_StateMachine_0.Lean_Into.AnimStateTransitionNode_4.Transition
    src=Lean_Back_Into
    via=AnimStateTransitionNode_4
   }
   {
    dst=Lean_Left_Into
    rule=AnimGraph.AnimGraphNode_StateMachine_1.Lean Transitions.AnimStateNode_5.Lean_Into.AnimGraphNode_StateMachine_0.Lean_Into.AnimStateTransitionNode_15.Transition
    src=Lean_Back_Into
    via=AnimStateTransitionNode_15
   }
   {
    dst=Lean_Right_Into
    rule=AnimGraph.AnimGraphNode_StateMachine_1.Lean Transitions.AnimStateNode_5.Lean_Into.AnimGraphNode_StateMachine_0.Lean_Into.AnimStateTransitionNode_8.Transition
    src=Lean_Back_Into
    via=AnimStateTransitionNode_8
   }
   {
    dst=Lean_Back_Into
    rule=AnimGraph.AnimGraphNode_StateMachine_1.Lean Transitions.AnimStateNode_5.Lean_Into.AnimGraphNode_StateMachine_0.Lean_Into.AnimStateTransitionNode_11.Transition
    src=Lean_Forward_Into
    via=AnimStateTransitionNode_11
   }
   {
    dst=Lean_Left_Into
    rule=AnimGraph.AnimGraphNode_StateMachine_1.Lean Transitions.AnimStateNode_5.Lean_Into.AnimGraphNode_StateMachine_0.Lean_Into.AnimStateTransitionNode_13.Transition
    src=Lean_Forward_Into
    via=AnimStateTransitionNode_13
   }
   {
    dst=Lean_Right_Into
    rule=AnimGraph.AnimGraphNode_StateMachine_1.Lean Transitions.AnimStateNode_5.Lean_Into.AnimGraphNode_StateMachine_0.Lean_Into.AnimStateTransitionNode_7.Transition
    src=Lean_Forward_Into
    via=AnimStateTransitionNode_7
   }
   {
    dst=Lean_Back_Into
    rule=AnimGraph.AnimGraphNode_StateMachine_1.Lean Transitions.AnimStateNode_5.Lean_Into.AnimGraphNode_StateMachine_0.Lean_Into.AnimStateTransitionNode_10.Transition
    src=Lean_Left_Into
    via=AnimStateTransitionNode_10
   }
   {
    dst=Lean_Forward_Into
    rule=AnimGraph.AnimGraphNode_StateMachine_1.Lean Transitions.AnimStateNode_5.Lean_Into.AnimGraphNode_StateMachine_0.Lean_Into.AnimStateTransitionNode_6.Transition
    src=Lean_Left_Into
    via=AnimStateTransitionNode_6
   }
   {
    dst=Lean_Right_Into
    rule=AnimGraph.AnimGraphNode_StateMachine_1.Lean Transitions.AnimStateNode_5.Lean_Into.AnimGraphNode_StateMachine_0.Lean_Into.AnimStateTransitionNode_9.Transition
    src=Lean_Left_Into
    via=AnimStateTransitionNode_9
   }
   {
    dst=Lean_Back_Into
    rule=AnimGraph.AnimGraphNode_StateMachine_1.Lean Transitions.AnimStateNode_5.Lean_Into.AnimGraphNode_StateMachine_0.Lean_Into.AnimStateTransitionNode_12.Transition
    src=Lean_Right_Into
    via=AnimStateTransitionNode_12
   }
   {
    dst=Lean_Forward_Into
    rule=AnimGraph.AnimGraphNode_StateMachine_1.Lean Transitions.AnimStateNode_5.Lean_Into.AnimGraphNode_StateMachine_0.Lean_Into.AnimStateTransitionNode_5.Transition
    src=Lean_Right_Into
    via=AnimStateTransitionNode_5
   }
   {
    dst=Lean_Left_Into
    rule=AnimGraph.AnimGraphNode_StateMachine_1.Lean Transitions.AnimStateNode_5.Lean_Into.AnimGraphNode_StateMachine_0.Lean_Into.AnimStateTransitionNode_14.Transition
    src=Lean_Right_Into
    via=AnimStateTransitionNode_14
   }
   {
    dst=Lean_Back_Into
    rule=AnimGraph.AnimGraphNode_StateMachine_1.Lean Transitions.AnimStateNode_5.Lean_Into.AnimGraphNode_StateMachine_0.Lean_Into.AnimStateTransitionNode_1.AnimationTransitionGraph_1
    src=Zero
    via=AnimStateTransitionNode_1
   }
   {
    dst=Lean_Forward_Into
    rule=AnimGraph.AnimGraphNode_StateMachine_1.Lean Transitions.AnimStateNode_5.Lean_Into.AnimGraphNode_StateMachine_0.Lean_Into.AnimStateTransitionNode_0.AnimationTransitionGraph_1
    src=Zero
    via=AnimStateTransitionNode_0
   }
   {
    dst=Lean_Left_Into
    rule=AnimGraph.AnimGraphNode_StateMachine_1.Lean Transitions.AnimStateNode_5.Lean_Into.AnimGraphNode_StateMachine_0.Lean_Into.AnimStateTransitionNode_2.AnimationTransitionGraph_1
    src=Zero
    via=AnimStateTransitionNode_2
   }
   {
    dst=Lean_Right_Into
    rule=AnimGraph.AnimGraphNode_StateMachine_1.Lean Transitions.AnimStateNode_5.Lean_Into.AnimGraphNode_StateMachine_0.Lean_Into.AnimStateTransitionNode_3.AnimationTransitionGraph_1
    src=Zero
    via=AnimStateTransitionNode_3
   }
  ]
 }
 Zero {
  path=AnimGraph.AnimGraphNode_StateMachine_1.Lean Transitions.AnimStateNode_5.Lean_Into.AnimGraphNode_StateMachine_0.Lean_Into.AnimStateNode_0.Zero
  nodes [Output Animation Pose(Result=Additive Identity Pose())]
 }
 Lean_Forward_Into {
  path=AnimGraph.AnimGraphNode_StateMachine_1.Lean Transitions.AnimStateNode_5.Lean_Into.AnimGraphNode_StateMachine_0.Lean_Into.AnimStateNode_1.Lean_Forward_Into
  nodes [
   n1 = "Sequence Player (additive) 'Season39_Lean_Forward_Into'"() {bLoopAnimation=false, sequence=/BRCosmetics/Gadgets/Assets/VinderTech_GliderChute/Glider_Season_39/Animation/Season39_Lean_Forward_Into}
   Output Animation Pose(Result=@n1)
  ]
 }
 Lean_Back_Into {
  path=AnimGraph.AnimGraphNode_StateMachine_1.Lean Transitions.AnimStateNode_5.Lean_Into.AnimGraphNode_StateMachine_0.Lean_Into.AnimStateNode_2.Lean_Back_Into
  nodes [
   n1 = "Sequence Player (additive) 'Season39_Lean_Back_Into'"() {bLoopAnimation=false, sequence=/BRCosmetics/Gadgets/Assets/VinderTech_GliderChute/Glider_Season_39/Animation/Season39_Lean_Back_Into}
   Output Animation Pose(Result=@n1)
  ]
 }
 Lean_Left_Into {
  path=AnimGraph.AnimGraphNode_StateMachine_1.Lean Transitions.AnimStateNode_5.Lean_Into.AnimGraphNode_StateMachine_0.Lean_Into.AnimStateNode_3.Lean_Left_Into
  nodes [
   n1 = "Sequence Player (additive) 'Season39_Lean_Left_Into'"() {bLoopAnimation=false, sequence=/BRCosmetics/Gadgets/Assets/VinderTech_GliderChute/Glider_Season_39/Animation/Season39_Lean_Left_Into}
   Output Animation Pose(Result=@n1)
  ]
 }
 Lean_Right_Into {
  path=AnimGraph.AnimGraphNode_StateMachine_1.Lean Transitions.AnimStateNode_5.Lean_Into.AnimGraphNode_StateMachine_0.Lean_Into.AnimStateNode_4.Lean_Right_Into
  nodes [
   n1 = "Sequence Player (additive) 'Season39_Lean_Right_Into'"() {bLoopAnimation=false, sequence=/BRCosmetics/Gadgets/Assets/VinderTech_GliderChute/Glider_Season_39/Animation/Season39_Lean_Right_Into}
   Output Animation Pose(Result=@n1)
  ]
 }
 AnimationTransitionGraph_1 {
  path=AnimGraph.AnimGraphNode_StateMachine_1.Lean Transitions.AnimStateNode_5.Lean_Into.AnimGraphNode_StateMachine_0.Lean_Into.AnimStateTransitionNode_0.AnimationTransitionGraph_1
  nodes [Result(bCanEnterTransition=Break Fort Anim Input Skydiving(FortAnimInput_Skydiving=Get SkydivingState()) -> (
     ParaGilderRootOffset=~
     ParaGliderRootRotationOverride=~
     bUseParaGlideRootModifier=~
     bIsSkydivingFromLaunchPad=~
     bIsSkydivingFromBus=~
     bIsInVortex=~
     bIsDiving=~
     bIsParachuteOpen=~
     bIsSkydiving=~
     bIsParachuteLeaning=~
     bIsSkydiveLeaning=~
     bIsLeaning=~
     bIsSkydiveDiveMode=~
     bPlaySkydiveDrift=~
     bSkydiveDriftDelayActive=~
     bSkydiveDriftAnimAllowed=~
     bIsGliderRight=~
     bIsGliderCenter=~
     bIsGliderLeft=~
     bIsGliderForward=~
     bIsGliderBack=~
     bHasDeployedAndIsStandingUpright=~
     LocalAccelForward=~
     LocalAccelRight=~
     LocalVelocityRight=~
     SkydiveAimPitch=~
     SkydiveAimYaw=~
     DeployChuteAnimRate=~
     SkydiveDriftAnimRate=~
     SkydiveDriftAnimRateCurrent=~
     SkydiveFidgetAnimRate=~
     SkydiveFidgetAnimRateCurrent=~
     SkydiveAdditiveAlpha=~
     SkydiveDriftDelay=~
    ).bIsGliderForward)]
 }
 AnimationTransitionGraph_1 {
  path=AnimGraph.AnimGraphNode_StateMachine_1.Lean Transitions.AnimStateNode_5.Lean_Into.AnimGraphNode_StateMachine_0.Lean_Into.AnimStateTransitionNode_1.AnimationTransitionGraph_1
  nodes [Result(bCanEnterTransition=Break Fort Anim Input Skydiving(FortAnimInput_Skydiving=Get SkydivingState()) -> (
     ParaGilderRootOffset=~
     ParaGliderRootRotationOverride=~
     bUseParaGlideRootModifier=~
     bIsSkydivingFromLaunchPad=~
     bIsSkydivingFromBus=~
     bIsInVortex=~
     bIsDiving=~
     bIsParachuteOpen=~
     bIsSkydiving=~
     bIsParachuteLeaning=~
     bIsSkydiveLeaning=~
     bIsLeaning=~
     bIsSkydiveDiveMode=~
     bPlaySkydiveDrift=~
     bSkydiveDriftDelayActive=~
     bSkydiveDriftAnimAllowed=~
     bIsGliderRight=~
     bIsGliderCenter=~
     bIsGliderLeft=~
     bIsGliderForward=~
     bIsGliderBack=~
     bHasDeployedAndIsStandingUpright=~
     LocalAccelForward=~
     LocalAccelRight=~
     LocalVelocityRight=~
     SkydiveAimPitch=~
     SkydiveAimYaw=~
     DeployChuteAnimRate=~
     SkydiveDriftAnimRate=~
     SkydiveDriftAnimRateCurrent=~
     SkydiveFidgetAnimRate=~
     SkydiveFidgetAnimRateCurrent=~
     SkydiveAdditiveAlpha=~
     SkydiveDriftDelay=~
    ).bIsGliderBack)]
 }
 Transition {
  path=AnimGraph.AnimGraphNode_StateMachine_1.Lean Transitions.AnimStateNode_5.Lean_Into.AnimGraphNode_StateMachine_0.Lean_Into.AnimStateTransitionNode_10.Transition
  nodes [Result(bCanEnterTransition=Break Fort Anim Input Skydiving(FortAnimInput_Skydiving=Get SkydivingState()) -> (
     ParaGilderRootOffset=~
     ParaGliderRootRotationOverride=~
     bUseParaGlideRootModifier=~
     bIsSkydivingFromLaunchPad=~
     bIsSkydivingFromBus=~
     bIsInVortex=~
     bIsDiving=~
     bIsParachuteOpen=~
     bIsSkydiving=~
     bIsParachuteLeaning=~
     bIsSkydiveLeaning=~
     bIsLeaning=~
     bIsSkydiveDiveMode=~
     bPlaySkydiveDrift=~
     bSkydiveDriftDelayActive=~
     bSkydiveDriftAnimAllowed=~
     bIsGliderRight=~
     bIsGliderCenter=~
     bIsGliderLeft=~
     bIsGliderForward=~
     bIsGliderBack=~
     bHasDeployedAndIsStandingUpright=~
     LocalAccelForward=~
     LocalAccelRight=~
     LocalVelocityRight=~
     SkydiveAimPitch=~
     SkydiveAimYaw=~
     DeployChuteAnimRate=~
     SkydiveDriftAnimRate=~
     SkydiveDriftAnimRateCurrent=~
     SkydiveFidgetAnimRate=~
     SkydiveFidgetAnimRateCurrent=~
     SkydiveAdditiveAlpha=~
     SkydiveDriftDelay=~
    ).bIsGliderBack)]
 }
 Transition {
  path=AnimGraph.AnimGraphNode_StateMachine_1.Lean Transitions.AnimStateNode_5.Lean_Into.AnimGraphNode_StateMachine_0.Lean_Into.AnimStateTransitionNode_11.Transition
  nodes [Result(bCanEnterTransition=Break Fort Anim Input Skydiving(FortAnimInput_Skydiving=Get SkydivingState()) -> (
     ParaGilderRootOffset=~
     ParaGliderRootRotationOverride=~
     bUseParaGlideRootModifier=~
     bIsSkydivingFromLaunchPad=~
     bIsSkydivingFromBus=~
     bIsInVortex=~
     bIsDiving=~
     bIsParachuteOpen=~
     bIsSkydiving=~
     bIsParachuteLeaning=~
     bIsSkydiveLeaning=~
     bIsLeaning=~
     bIsSkydiveDiveMode=~
     bPlaySkydiveDrift=~
     bSkydiveDriftDelayActive=~
     bSkydiveDriftAnimAllowed=~
     bIsGliderRight=~
     bIsGliderCenter=~
     bIsGliderLeft=~
     bIsGliderForward=~
     bIsGliderBack=~
     bHasDeployedAndIsStandingUpright=~
     LocalAccelForward=~
     LocalAccelRight=~
     LocalVelocityRight=~
     SkydiveAimPitch=~
     SkydiveAimYaw=~
     DeployChuteAnimRate=~
     SkydiveDriftAnimRate=~
     SkydiveDriftAnimRateCurrent=~
     SkydiveFidgetAnimRate=~
     SkydiveFidgetAnimRateCurrent=~
     SkydiveAdditiveAlpha=~
     SkydiveDriftDelay=~
    ).bIsGliderBack)]
 }
 Transition {
  path=AnimGraph.AnimGraphNode_StateMachine_1.Lean Transitions.AnimStateNode_5.Lean_Into.AnimGraphNode_StateMachine_0.Lean_Into.AnimStateTransitionNode_12.Transition
  nodes [Result(bCanEnterTransition=Break Fort Anim Input Skydiving(FortAnimInput_Skydiving=Get SkydivingState()) -> (
     ParaGilderRootOffset=~
     ParaGliderRootRotationOverride=~
     bUseParaGlideRootModifier=~
     bIsSkydivingFromLaunchPad=~
     bIsSkydivingFromBus=~
     bIsInVortex=~
     bIsDiving=~
     bIsParachuteOpen=~
     bIsSkydiving=~
     bIsParachuteLeaning=~
     bIsSkydiveLeaning=~
     bIsLeaning=~
     bIsSkydiveDiveMode=~
     bPlaySkydiveDrift=~
     bSkydiveDriftDelayActive=~
     bSkydiveDriftAnimAllowed=~
     bIsGliderRight=~
     bIsGliderCenter=~
     bIsGliderLeft=~
     bIsGliderForward=~
     bIsGliderBack=~
     bHasDeployedAndIsStandingUpright=~
     LocalAccelForward=~
     LocalAccelRight=~
     LocalVelocityRight=~
     SkydiveAimPitch=~
     SkydiveAimYaw=~
     DeployChuteAnimRate=~
     SkydiveDriftAnimRate=~
     SkydiveDriftAnimRateCurrent=~
     SkydiveFidgetAnimRate=~
     SkydiveFidgetAnimRateCurrent=~
     SkydiveAdditiveAlpha=~
     SkydiveDriftDelay=~
    ).bIsGliderBack)]
 }
 Transition {
  path=AnimGraph.AnimGraphNode_StateMachine_1.Lean Transitions.AnimStateNode_5.Lean_Into.AnimGraphNode_StateMachine_0.Lean_Into.AnimStateTransitionNode_13.Transition
  nodes [Result(bCanEnterTransition=Break Fort Anim Input Skydiving(FortAnimInput_Skydiving=Get SkydivingState()) -> (
     ParaGilderRootOffset=~
     ParaGliderRootRotationOverride=~
     bUseParaGlideRootModifier=~
     bIsSkydivingFromLaunchPad=~
     bIsSkydivingFromBus=~
     bIsInVortex=~
     bIsDiving=~
     bIsParachuteOpen=~
     bIsSkydiving=~
     bIsParachuteLeaning=~
     bIsSkydiveLeaning=~
     bIsLeaning=~
     bIsSkydiveDiveMode=~
     bPlaySkydiveDrift=~
     bSkydiveDriftDelayActive=~
     bSkydiveDriftAnimAllowed=~
     bIsGliderRight=~
     bIsGliderCenter=~
     bIsGliderLeft=~
     bIsGliderForward=~
     bIsGliderBack=~
     bHasDeployedAndIsStandingUpright=~
     LocalAccelForward=~
     LocalAccelRight=~
     LocalVelocityRight=~
     SkydiveAimPitch=~
     SkydiveAimYaw=~
     DeployChuteAnimRate=~
     SkydiveDriftAnimRate=~
     SkydiveDriftAnimRateCurrent=~
     SkydiveFidgetAnimRate=~
     SkydiveFidgetAnimRateCurrent=~
     SkydiveAdditiveAlpha=~
     SkydiveDriftDelay=~
    ).bIsGliderLeft)]
 }
 Transition {
  path=AnimGraph.AnimGraphNode_StateMachine_1.Lean Transitions.AnimStateNode_5.Lean_Into.AnimGraphNode_StateMachine_0.Lean_Into.AnimStateTransitionNode_14.Transition
  nodes [Result(bCanEnterTransition=Break Fort Anim Input Skydiving(FortAnimInput_Skydiving=Get SkydivingState()) -> (
     ParaGilderRootOffset=~
     ParaGliderRootRotationOverride=~
     bUseParaGlideRootModifier=~
     bIsSkydivingFromLaunchPad=~
     bIsSkydivingFromBus=~
     bIsInVortex=~
     bIsDiving=~
     bIsParachuteOpen=~
     bIsSkydiving=~
     bIsParachuteLeaning=~
     bIsSkydiveLeaning=~
     bIsLeaning=~
     bIsSkydiveDiveMode=~
     bPlaySkydiveDrift=~
     bSkydiveDriftDelayActive=~
     bSkydiveDriftAnimAllowed=~
     bIsGliderRight=~
     bIsGliderCenter=~
     bIsGliderLeft=~
     bIsGliderForward=~
     bIsGliderBack=~
     bHasDeployedAndIsStandingUpright=~
     LocalAccelForward=~
     LocalAccelRight=~
     LocalVelocityRight=~
     SkydiveAimPitch=~
     SkydiveAimYaw=~
     DeployChuteAnimRate=~
     SkydiveDriftAnimRate=~
     SkydiveDriftAnimRateCurrent=~
     SkydiveFidgetAnimRate=~
     SkydiveFidgetAnimRateCurrent=~
     SkydiveAdditiveAlpha=~
     SkydiveDriftDelay=~
    ).bIsGliderLeft)]
 }
 Transition {
  path=AnimGraph.AnimGraphNode_StateMachine_1.Lean Transitions.AnimStateNode_5.Lean_Into.AnimGraphNode_StateMachine_0.Lean_Into.AnimStateTransitionNode_15.Transition
  nodes [Result(bCanEnterTransition=Break Fort Anim Input Skydiving(FortAnimInput_Skydiving=Get SkydivingState()) -> (
     ParaGilderRootOffset=~
     ParaGliderRootRotationOverride=~
     bUseParaGlideRootModifier=~
     bIsSkydivingFromLaunchPad=~
     bIsSkydivingFromBus=~
     bIsInVortex=~
     bIsDiving=~
     bIsParachuteOpen=~
     bIsSkydiving=~
     bIsParachuteLeaning=~
     bIsSkydiveLeaning=~
     bIsLeaning=~
     bIsSkydiveDiveMode=~
     bPlaySkydiveDrift=~
     bSkydiveDriftDelayActive=~
     bSkydiveDriftAnimAllowed=~
     bIsGliderRight=~
     bIsGliderCenter=~
     bIsGliderLeft=~
     bIsGliderForward=~
     bIsGliderBack=~
     bHasDeployedAndIsStandingUpright=~
     LocalAccelForward=~
     LocalAccelRight=~
     LocalVelocityRight=~
     SkydiveAimPitch=~
     SkydiveAimYaw=~
     DeployChuteAnimRate=~
     SkydiveDriftAnimRate=~
     SkydiveDriftAnimRateCurrent=~
     SkydiveFidgetAnimRate=~
     SkydiveFidgetAnimRateCurrent=~
     SkydiveAdditiveAlpha=~
     SkydiveDriftDelay=~
    ).bIsGliderLeft)]
 }
 AnimationTransitionGraph_1 {
  path=AnimGraph.AnimGraphNode_StateMachine_1.Lean Transitions.AnimStateNode_5.Lean_Into.AnimGraphNode_StateMachine_0.Lean_Into.AnimStateTransitionNode_2.AnimationTransitionGraph_1
  nodes [Result(bCanEnterTransition=Break Fort Anim Input Skydiving(FortAnimInput_Skydiving=Get SkydivingState()) -> (
     ParaGilderRootOffset=~
     ParaGliderRootRotationOverride=~
     bUseParaGlideRootModifier=~
     bIsSkydivingFromLaunchPad=~
     bIsSkydivingFromBus=~
     bIsInVortex=~
     bIsDiving=~
     bIsParachuteOpen=~
     bIsSkydiving=~
     bIsParachuteLeaning=~
     bIsSkydiveLeaning=~
     bIsLeaning=~
     bIsSkydiveDiveMode=~
     bPlaySkydiveDrift=~
     bSkydiveDriftDelayActive=~
     bSkydiveDriftAnimAllowed=~
     bIsGliderRight=~
     bIsGliderCenter=~
     bIsGliderLeft=~
     bIsGliderForward=~
     bIsGliderBack=~
     bHasDeployedAndIsStandingUpright=~
     LocalAccelForward=~
     LocalAccelRight=~
     LocalVelocityRight=~
     SkydiveAimPitch=~
     SkydiveAimYaw=~
     DeployChuteAnimRate=~
     SkydiveDriftAnimRate=~
     SkydiveDriftAnimRateCurrent=~
     SkydiveFidgetAnimRate=~
     SkydiveFidgetAnimRateCurrent=~
     SkydiveAdditiveAlpha=~
     SkydiveDriftDelay=~
    ).bIsGliderLeft)]
 }
 AnimationTransitionGraph_1 {
  path=AnimGraph.AnimGraphNode_StateMachine_1.Lean Transitions.AnimStateNode_5.Lean_Into.AnimGraphNode_StateMachine_0.Lean_Into.AnimStateTransitionNode_3.AnimationTransitionGraph_1
  nodes [Result(bCanEnterTransition=Break Fort Anim Input Skydiving(FortAnimInput_Skydiving=Get SkydivingState()) -> (
     ParaGilderRootOffset=~
     ParaGliderRootRotationOverride=~
     bUseParaGlideRootModifier=~
     bIsSkydivingFromLaunchPad=~
     bIsSkydivingFromBus=~
     bIsInVortex=~
     bIsDiving=~
     bIsParachuteOpen=~
     bIsSkydiving=~
     bIsParachuteLeaning=~
     bIsSkydiveLeaning=~
     bIsLeaning=~
     bIsSkydiveDiveMode=~
     bPlaySkydiveDrift=~
     bSkydiveDriftDelayActive=~
     bSkydiveDriftAnimAllowed=~
     bIsGliderRight=~
     bIsGliderCenter=~
     bIsGliderLeft=~
     bIsGliderForward=~
     bIsGliderBack=~
     bHasDeployedAndIsStandingUpright=~
     LocalAccelForward=~
     LocalAccelRight=~
     LocalVelocityRight=~
     SkydiveAimPitch=~
     SkydiveAimYaw=~
     DeployChuteAnimRate=~
     SkydiveDriftAnimRate=~
     SkydiveDriftAnimRateCurrent=~
     SkydiveFidgetAnimRate=~
     SkydiveFidgetAnimRateCurrent=~
     SkydiveAdditiveAlpha=~
     SkydiveDriftDelay=~
    ).bIsGliderRight)]
 }
 Transition {
  path=AnimGraph.AnimGraphNode_StateMachine_1.Lean Transitions.AnimStateNode_5.Lean_Into.AnimGraphNode_StateMachine_0.Lean_Into.AnimStateTransitionNode_4.Transition
  nodes [Result(bCanEnterTransition=Break Fort Anim Input Skydiving(FortAnimInput_Skydiving=Get SkydivingState()) -> (
     ParaGilderRootOffset=~
     ParaGliderRootRotationOverride=~
     bUseParaGlideRootModifier=~
     bIsSkydivingFromLaunchPad=~
     bIsSkydivingFromBus=~
     bIsInVortex=~
     bIsDiving=~
     bIsParachuteOpen=~
     bIsSkydiving=~
     bIsParachuteLeaning=~
     bIsSkydiveLeaning=~
     bIsLeaning=~
     bIsSkydiveDiveMode=~
     bPlaySkydiveDrift=~
     bSkydiveDriftDelayActive=~
     bSkydiveDriftAnimAllowed=~
     bIsGliderRight=~
     bIsGliderCenter=~
     bIsGliderLeft=~
     bIsGliderForward=~
     bIsGliderBack=~
     bHasDeployedAndIsStandingUpright=~
     LocalAccelForward=~
     LocalAccelRight=~
     LocalVelocityRight=~
     SkydiveAimPitch=~
     SkydiveAimYaw=~
     DeployChuteAnimRate=~
     SkydiveDriftAnimRate=~
     SkydiveDriftAnimRateCurrent=~
     SkydiveFidgetAnimRate=~
     SkydiveFidgetAnimRateCurrent=~
     SkydiveAdditiveAlpha=~
     SkydiveDriftDelay=~
    ).bIsGliderForward)]
 }
 Transition {
  path=AnimGraph.AnimGraphNode_StateMachine_1.Lean Transitions.AnimStateNode_5.Lean_Into.AnimGraphNode_StateMachine_0.Lean_Into.AnimStateTransitionNode_5.Transition
  nodes [Result(bCanEnterTransition=Break Fort Anim Input Skydiving(FortAnimInput_Skydiving=Get SkydivingState()) -> (
     ParaGilderRootOffset=~
     ParaGliderRootRotationOverride=~
     bUseParaGlideRootModifier=~
     bIsSkydivingFromLaunchPad=~
     bIsSkydivingFromBus=~
     bIsInVortex=~
     bIsDiving=~
     bIsParachuteOpen=~
     bIsSkydiving=~
     bIsParachuteLeaning=~
     bIsSkydiveLeaning=~
     bIsLeaning=~
     bIsSkydiveDiveMode=~
     bPlaySkydiveDrift=~
     bSkydiveDriftDelayActive=~
     bSkydiveDriftAnimAllowed=~
     bIsGliderRight=~
     bIsGliderCenter=~
     bIsGliderLeft=~
     bIsGliderForward=~
     bIsGliderBack=~
     bHasDeployedAndIsStandingUpright=~
     LocalAccelForward=~
     LocalAccelRight=~
     LocalVelocityRight=~
     SkydiveAimPitch=~
     SkydiveAimYaw=~
     DeployChuteAnimRate=~
     SkydiveDriftAnimRate=~
     SkydiveDriftAnimRateCurrent=~
     SkydiveFidgetAnimRate=~
     SkydiveFidgetAnimRateCurrent=~
     SkydiveAdditiveAlpha=~
     SkydiveDriftDelay=~
    ).bIsGliderForward)]
 }
 Transition {
  path=AnimGraph.AnimGraphNode_StateMachine_1.Lean Transitions.AnimStateNode_5.Lean_Into.AnimGraphNode_StateMachine_0.Lean_Into.AnimStateTransitionNode_6.Transition
  nodes [Result(bCanEnterTransition=Break Fort Anim Input Skydiving(FortAnimInput_Skydiving=Get SkydivingState()) -> (
     ParaGilderRootOffset=~
     ParaGliderRootRotationOverride=~
     bUseParaGlideRootModifier=~
     bIsSkydivingFromLaunchPad=~
     bIsSkydivingFromBus=~
     bIsInVortex=~
     bIsDiving=~
     bIsParachuteOpen=~
     bIsSkydiving=~
     bIsParachuteLeaning=~
     bIsSkydiveLeaning=~
     bIsLeaning=~
     bIsSkydiveDiveMode=~
     bPlaySkydiveDrift=~
     bSkydiveDriftDelayActive=~
     bSkydiveDriftAnimAllowed=~
     bIsGliderRight=~
     bIsGliderCenter=~
     bIsGliderLeft=~
     bIsGliderForward=~
     bIsGliderBack=~
     bHasDeployedAndIsStandingUpright=~
     LocalAccelForward=~
     LocalAccelRight=~
     LocalVelocityRight=~
     SkydiveAimPitch=~
     SkydiveAimYaw=~
     DeployChuteAnimRate=~
     SkydiveDriftAnimRate=~
     SkydiveDriftAnimRateCurrent=~
     SkydiveFidgetAnimRate=~
     SkydiveFidgetAnimRateCurrent=~
     SkydiveAdditiveAlpha=~
     SkydiveDriftDelay=~
    ).bIsGliderForward)]
 }
 Transition {
  path=AnimGraph.AnimGraphNode_StateMachine_1.Lean Transitions.AnimStateNode_5.Lean_Into.AnimGraphNode_StateMachine_0.Lean_Into.AnimStateTransitionNode_7.Transition
  nodes [Result(bCanEnterTransition=Break Fort Anim Input Skydiving(FortAnimInput_Skydiving=Get SkydivingState()) -> (
     ParaGilderRootOffset=~
     ParaGliderRootRotationOverride=~
     bUseParaGlideRootModifier=~
     bIsSkydivingFromLaunchPad=~
     bIsSkydivingFromBus=~
     bIsInVortex=~
     bIsDiving=~
     bIsParachuteOpen=~
     bIsSkydiving=~
     bIsParachuteLeaning=~
     bIsSkydiveLeaning=~
     bIsLeaning=~
     bIsSkydiveDiveMode=~
     bPlaySkydiveDrift=~
     bSkydiveDriftDelayActive=~
     bSkydiveDriftAnimAllowed=~
     bIsGliderRight=~
     bIsGliderCenter=~
     bIsGliderLeft=~
     bIsGliderForward=~
     bIsGliderBack=~
     bHasDeployedAndIsStandingUpright=~
     LocalAccelForward=~
     LocalAccelRight=~
     LocalVelocityRight=~
     SkydiveAimPitch=~
     SkydiveAimYaw=~
     DeployChuteAnimRate=~
     SkydiveDriftAnimRate=~
     SkydiveDriftAnimRateCurrent=~
     SkydiveFidgetAnimRate=~
     SkydiveFidgetAnimRateCurrent=~
     SkydiveAdditiveAlpha=~
     SkydiveDriftDelay=~
    ).bIsGliderRight)]
 }
 Transition {
  path=AnimGraph.AnimGraphNode_StateMachine_1.Lean Transitions.AnimStateNode_5.Lean_Into.AnimGraphNode_StateMachine_0.Lean_Into.AnimStateTransitionNode_8.Transition
  nodes [Result(bCanEnterTransition=Break Fort Anim Input Skydiving(FortAnimInput_Skydiving=Get SkydivingState()) -> (
     ParaGilderRootOffset=~
     ParaGliderRootRotationOverride=~
     bUseParaGlideRootModifier=~
     bIsSkydivingFromLaunchPad=~
     bIsSkydivingFromBus=~
     bIsInVortex=~
     bIsDiving=~
     bIsParachuteOpen=~
     bIsSkydiving=~
     bIsParachuteLeaning=~
     bIsSkydiveLeaning=~
     bIsLeaning=~
     bIsSkydiveDiveMode=~
     bPlaySkydiveDrift=~
     bSkydiveDriftDelayActive=~
     bSkydiveDriftAnimAllowed=~
     bIsGliderRight=~
     bIsGliderCenter=~
     bIsGliderLeft=~
     bIsGliderForward=~
     bIsGliderBack=~
     bHasDeployedAndIsStandingUpright=~
     LocalAccelForward=~
     LocalAccelRight=~
     LocalVelocityRight=~
     SkydiveAimPitch=~
     SkydiveAimYaw=~
     DeployChuteAnimRate=~
     SkydiveDriftAnimRate=~
     SkydiveDriftAnimRateCurrent=~
     SkydiveFidgetAnimRate=~
     SkydiveFidgetAnimRateCurrent=~
     SkydiveAdditiveAlpha=~
     SkydiveDriftDelay=~
    ).bIsGliderRight)]
 }
 Transition {
  path=AnimGraph.AnimGraphNode_StateMachine_1.Lean Transitions.AnimStateNode_5.Lean_Into.AnimGraphNode_StateMachine_0.Lean_Into.AnimStateTransitionNode_9.Transition
  nodes [Result(bCanEnterTransition=Break Fort Anim Input Skydiving(FortAnimInput_Skydiving=Get SkydivingState()) -> (
     ParaGilderRootOffset=~
     ParaGliderRootRotationOverride=~
     bUseParaGlideRootModifier=~
     bIsSkydivingFromLaunchPad=~
     bIsSkydivingFromBus=~
     bIsInVortex=~
     bIsDiving=~
     bIsParachuteOpen=~
     bIsSkydiving=~
     bIsParachuteLeaning=~
     bIsSkydiveLeaning=~
     bIsLeaning=~
     bIsSkydiveDiveMode=~
     bPlaySkydiveDrift=~
     bSkydiveDriftDelayActive=~
     bSkydiveDriftAnimAllowed=~
     bIsGliderRight=~
     bIsGliderCenter=~
     bIsGliderLeft=~
     bIsGliderForward=~
     bIsGliderBack=~
     bHasDeployedAndIsStandingUpright=~
     LocalAccelForward=~
     LocalAccelRight=~
     LocalVelocityRight=~
     SkydiveAimPitch=~
     SkydiveAimYaw=~
     DeployChuteAnimRate=~
     SkydiveDriftAnimRate=~
     SkydiveDriftAnimRateCurrent=~
     SkydiveFidgetAnimRate=~
     SkydiveFidgetAnimRateCurrent=~
     SkydiveAdditiveAlpha=~
     SkydiveDriftDelay=~
    ).bIsGliderRight)]
 }
 AnimationTransitionGraph_0 {
  path=AnimGraph.AnimGraphNode_StateMachine_1.Lean Transitions.AnimStateTransitionNode_0.AnimationTransitionGraph_0
  nodes [Result(bCanEnterTransition=Break Fort Anim Input Skydiving(FortAnimInput_Skydiving=Get SkydivingState()) -> (
     ParaGilderRootOffset=~
     ParaGliderRootRotationOverride=~
     bUseParaGlideRootModifier=~
     bIsSkydivingFromLaunchPad=~
     bIsSkydivingFromBus=~
     bIsInVortex=~
     bIsDiving=~
     bIsParachuteOpen=~
     bIsSkydiving=~
     bIsParachuteLeaning=~
     bIsSkydiveLeaning=~
     bIsLeaning=~
     bIsSkydiveDiveMode=~
     bPlaySkydiveDrift=~
     bSkydiveDriftDelayActive=~
     bSkydiveDriftAnimAllowed=~
     bIsGliderRight=~
     bIsGliderCenter=~
     bIsGliderLeft=~
     bIsGliderForward=~
     bIsGliderBack=~
     bHasDeployedAndIsStandingUpright=~
     LocalAccelForward=~
     LocalAccelRight=~
     LocalVelocityRight=~
     SkydiveAimPitch=~
     SkydiveAimYaw=~
     DeployChuteAnimRate=~
     SkydiveDriftAnimRate=~
     SkydiveDriftAnimRateCurrent=~
     SkydiveFidgetAnimRate=~
     SkydiveFidgetAnimRateCurrent=~
     SkydiveAdditiveAlpha=~
     SkydiveDriftDelay=~
    ).bIsLeaning)]
 }
 AnimationTransitionGraph_0 {
  path=AnimGraph.AnimGraphNode_StateMachine_1.Lean Transitions.AnimStateTransitionNode_11.AnimationTransitionGraph_0
  nodes [Result(bCanEnterTransition=NOT Boolean(A=Get bIsAccelerating2D()))]
 }
 AnimationTransitionGraph_0 {path=AnimGraph.AnimGraphNode_StateMachine_1.Lean Transitions.AnimStateTransitionNode_18.AnimationTransitionGraph_0, nodes [Result()]}
 AnimationTransitionGraph_0 {
  path=AnimGraph.AnimGraphNode_StateMachine_1.Lean Transitions.AnimStateTransitionNode_19.AnimationTransitionGraph_0
  nodes [Result(bCanEnterTransition=Get bIsAccelerating2D())]
 }
 TopSpin {
  entry=NotSpinning
  path=AnimGraph.AnimGraphNode_StateMachine_2.TopSpin
  ids=id
  nodes [
   AnimStateNode_0 = NotSpinning() {is=title AnimStateNode}
   AnimStateNode_1 = Spinning() {is=title AnimStateNode, bAlwaysResetOnEntry=true}
   AnimStateTransitionNode_0 = AnimStateTransitionNode() {bAllowInertializationForSelfTransitions=true}
   AnimStateTransitionNode_1 = AnimStateTransitionNode() {bAllowInertializationForSelfTransitions=true}
  ]
  links [
   {
    dst=Spinning
    rule=AnimGraph.AnimGraphNode_StateMachine_2.TopSpin.AnimStateTransitionNode_0.Transition
    src=NotSpinning
    via=AnimStateTransitionNode_0
   }
   {
    dst=NotSpinning
    rule=AnimGraph.AnimGraphNode_StateMachine_2.TopSpin.AnimStateTransitionNode_1.Transition
    src=Spinning
    via=AnimStateTransitionNode_1
   }
  ]
 }
 NotSpinning {
  path=AnimGraph.AnimGraphNode_StateMachine_2.TopSpin.AnimStateNode_0.NotSpinning
  nodes [Output Animation Pose(Result=Additive Identity Pose())]
 }
 Spinning {
  path=AnimGraph.AnimGraphNode_StateMachine_2.TopSpin.AnimStateNode_1.Spinning
  nodes [Output Animation Pose(Result="Sequence Player (additive) 'Season39_TopSpin_GLIDER'"!(PlayRate:float=Get TopSpin_Speed() -> (TopSpin_Speed:double)) -> (Pose) {sequence=/BRCosmetics/Gadgets/Assets/VinderTech_GliderChute/Glider_Season_39/Animation/Season39_TopSpin_GLIDER})]
 }
 Transition {
  path=AnimGraph.AnimGraphNode_StateMachine_2.TopSpin.AnimStateTransitionNode_0.Transition
  nodes [Result(bCanEnterTransition=Get StartSpin() -> (StartSpin:bool))]
 }
 Transition {
  path=AnimGraph.AnimGraphNode_StateMachine_2.TopSpin.AnimStateTransitionNode_1.Transition
  nodes [Result(bCanEnterTransition=NOT Boolean(A=Get StartSpin()))]
 }
 SideSpin {
  entry=NoSideSpin
  path=AnimGraph.AnimGraphNode_StateMachine_3.SideSpin
  ids=id
  nodes [
   AnimStateNode_0 = NoSideSpin() {is=title AnimStateNode}
   AnimStateNode_1 = SideSpin() {is=title AnimStateNode}
   AnimStateTransitionNode_0 = AnimStateTransitionNode() {bAllowInertializationForSelfTransitions=true}
   AnimStateTransitionNode_1 = AnimStateTransitionNode() {bAllowInertializationForSelfTransitions=true}
  ]
  links [
   {
    dst=SideSpin
    rule=AnimGraph.AnimGraphNode_StateMachine_3.SideSpin.AnimStateTransitionNode_0.Transition
    src=NoSideSpin
    via=AnimStateTransitionNode_0
   }
   {
    dst=NoSideSpin
    rule=AnimGraph.AnimGraphNode_StateMachine_3.SideSpin.AnimStateTransitionNode_1.Transition
    src=SideSpin
    via=AnimStateTransitionNode_1
   }
  ]
 }
 NoSideSpin {
  path=AnimGraph.AnimGraphNode_StateMachine_3.SideSpin.AnimStateNode_0.NoSideSpin
  nodes [Output Animation Pose(Result=Additive Identity Pose())]
 }
 SideSpin {
  path=AnimGraph.AnimGraphNode_StateMachine_3.SideSpin.AnimStateNode_1.SideSpin
  nodes [Output Animation Pose(Result="Sequence Player (additive) 'Season39_SideSpinners_GLIDER'"!(PlayRate:float=Get SideSpinner_Speed() -> (SideSpinner_Speed:double)) -> (Pose) {sequence=/BRCosmetics/Gadgets/Assets/VinderTech_GliderChute/Glider_Season_39/Animation/Season39_SideSpinners_GLIDER})]
 }
 Transition {
  path=AnimGraph.AnimGraphNode_StateMachine_3.SideSpin.AnimStateTransitionNode_0.Transition
  nodes [Result(bCanEnterTransition=Get StartSpin())]
 }
 Transition {
  path=AnimGraph.AnimGraphNode_StateMachine_3.SideSpin.AnimStateTransitionNode_1.Transition
  nodes [Result(bCanEnterTransition=NOT Boolean(A=Get StartSpin()))]
 }
 EventGraph {nodes [
   n1 = Event BlueprintUpdateAnimation()
   > Set Time(Time:double=@n1.DeltaTimeX) -> (Output_Get:double)
   > n2 = Sequence()
   > Get Player Open Glider Bool!(execute:Exec) -> (true:Exec) {is=title K2Node_MacroInstance}
   n3 = Break Fort Anim Input Skydiving(FortAnimInput_Skydiving=Get SkydivingState()) -> (
    ParaGilderRootOffset=~
    ParaGliderRootRotationOverride=~
    bUseParaGlideRootModifier=~
    bIsSkydivingFromLaunchPad=~
    bIsSkydivingFromBus=~
    bIsInVortex=~
    bIsDiving=~
    bIsParachuteOpen=~
    bIsSkydiving=~
    bIsParachuteLeaning=~
    bIsSkydiveLeaning=~
    bIsLeaning=~
    bIsSkydiveDiveMode=~
    bPlaySkydiveDrift=~
    bSkydiveDriftDelayActive=~
    bSkydiveDriftAnimAllowed=~
    bIsGliderRight=~
    bIsGliderCenter=~
    bIsGliderLeft=~
    bIsGliderForward=~
    bIsGliderBack=~
    bHasDeployedAndIsStandingUpright=~
    LocalAccelForward=~
    LocalAccelRight=~
    LocalVelocityRight=~
    SkydiveAimPitch=~
    SkydiveAimYaw=~
    DeployChuteAnimRate=~
    SkydiveDriftAnimRate=~
    SkydiveDriftAnimRateCurrent=~
    SkydiveFidgetAnimRate=~
    SkydiveFidgetAnimRateCurrent=~
    SkydiveAdditiveAlpha=~
    SkydiveDriftDelay=~
   )
   > Set IsParachuteOpen(IsParachuteOpen:bool=@n3.bIsParachuteOpen) -> (Output_Get:bool)
   > Set DeployAnimRate(DeployAnimRate:double=@n3.DeployChuteAnimRate) -> (Output_Get:double)
   n4 = Get StartJitter() -> (StartJitter:bool)
   n5 = NOT Boolean(A=Get StartSpin())
   n6 = Get IsCosmeticPreview(self=Get OwnerGlider())
   > n7 = Branch(Condition=OR Boolean!(A=OR Boolean(A=@n6, B=@n5), B=NOT Boolean(A=@n4)) -> (ReturnValue))
   > Set TopSpin_Speed(TopSpin_Speed:double=0) -> (Output_Get:double)
   > Set Jitter_Speed(Jitter_Speed:double=0) -> (Output_Get:double)
   Branch(execute=@n7.else, Condition=Get StartSpin())
   > Branch(Condition=Get StartJitter())
   > n8 = ForwardBack_Value_Interp!(
    execute
    Time:double=Get Time() -> (Time:double)
    Currnet_Value:double=Get TopSpin_Speed()
    Idle_Value:double=1
    Idle_InterpSpeed:double=1
    Forward_Value:double=3
    Forward_InterpSpeed:double=2
    Back_Value:double=-1.75
    Back_InterpSpeed:double=2
    LocalAccelForward:double=@n3.LocalAccelForward
   ) -> (then, New_Value:double)
   > n9 = Set TopSpin_Speed(TopSpin_Speed=@n8)
   > Set Jitter_Speed(Jitter_Speed=@n9)
   n10 = Branch(execute=@n2.then_1, Condition=@n3.bIsParachuteOpen)
   Set StartSpin(execute=@n10.else, StartSpin:bool=false) -> (Output_Get:bool)
   > Set StartJitter(StartJitter:bool=false) -> (Output_Get:bool)
   AnimNotify_TurnONEngines() {is=title K2Node_Event}
   > SetGliderFullyDeployed!(execute, self:Fort Glider Anim Instance Object Reference, bIsFullyDeployed:bool=true) -> (then)
   > Set StartJitter(StartJitter=true)
   AnimNotify_StartTopSpin() {is=title K2Node_Event}
   > SetGliderFullyDeployed(bIsFullyDeployed=true)
   > Set StartSpin(StartSpin=true)
  ]}
 GetPlayerOpenGliderBool {nodes [
   Inputs()
   > n1 = Branch(Condition=Get IsCosmeticPreview(self=Get OwnerGlider()))
   Outputs!(true:Exec=[@n1.then, @n3.then])
   n2 = Get OwnerFortPawn() -> (OwnerFortPawn:Fort Player Pawn Object Reference)
   n3 = Set bForceOpen(
    execute=@n1.else
    bForceOpen:bool=ShouldForceOpenGlider!(self:Fort Player Animation Interface Object Reference=GetAnimationInterface!(self:Fort Player Pawn Object Reference=@n2) -> (ReturnValue:Fort Player Animation Interface Object Reference)) -> (ReturnValue:bool)
   ) -> (Output_Get:bool)
  ]}
]
variables [
 DeployAnimRate {type=number}
 IsParachuteOpen {type=boolean}
 Jitter_Speed {type=number}
 SideSpinner_Speed {type=number, default=1}
 StartJitter {type=boolean}
 StartSpin {type=boolean}
 Time {type=number}
 TopSpin_Speed {type=number}
]
